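# Masked, lit gold banner material for a cumulative quest interactable, with wind-driven cloth motion and glow effects.
# Switchable view-space Fresnel glow and a panning sweep highlight, projected along a camera-space direction, are summed.
# Two panning flag normal samples, offset by actor position, drive UV-masked vertex sway and blended normals, with zeroed sway on ES2/ES3_1.
# An Opacity parameter offsets a noise texture for dissolve-style masking.
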Material /Game/Athena/Items/QuestInteractables/Cumulative/FX/M_GoldBanner_01
graph {
 properties {bUsedWithNanite=true, blendMode=BLEND_Masked, pixelDepthOffsetMode=PDOM_Legacy, usedShadingModels=Default Lit}
 nodes [
  n1 = VertexNormalWS()
  n2 = MaterialFunctionCall(
   Exponent=Constant() {r=4}
   Camera Vector=Constant3Vector() {constant=<rgba 0 0 -1 0>}
   Vertex / Pixel Normal=Normalize(VectorInput=Transform(None=@n1) {transformSourceType=TRANSFORMSOURCE_World, transformType=TRANSFORM_View})
  ) {materialFunction=&MF_Fresnel_FX}
  n3 = VectorParameter() {parameterName=GlowColour, defaultValue=<rgba 10 1.6667 0 0>}
  n4 = Multiply(
   A=Subtract(
    A=WorldPosition().XYZ
    B=MaterialFunctionCall() {materialFunction=&ComponentPivotLocation}.Result
   )
   B=ScalarParameter() {defaultValue=0.001, parameterName=WorldTileScale}
  )
  n5 = Normalize(VectorInput=Transform(None=VectorParameter() {parameterName=SweepDirectionCamSpace, defaultValue=<rgba -0.5 0.45 -1 0>}.RGB) {transformSourceType=TRANSFORMSOURCE_Camera})
  n6 = ComponentMask(None=Multiply(A=DotProduct(A=@n4, B=@n5), B=@n5)) {b=true, r=true}
  n7 = ScalarParameter() {defaultValue=0.28, parameterName=PanSpeed}
  n8 = Frac(None=ComponentMask(None=Panner(Coordinate=@n6, Speed=@n7) {speedX=0.2, speedY=0.2}) {g=true})
  n9 = Add(A=@n8) {constB=0.1}
  n10 = OneMinus(None=Multiply(A=Subtract(A=@n8) {constB=0.9}) {constB=10})
  n11 = MaterialFunctionCall() {materialFunction=&MF_Fresnel_FX}
  n12 = Multiply(A=Multiply(A=Floor(None=@n9), B=Power(Base=@n10) {constExponent=3}), B=Saturate(None=Add(A=@n11.Fresnel Inverted) {constB=0.1}))
  n13 = VectorParameter() {parameterName=SweepColour, defaultValue=<rgba 2 1.05 0.1 0>}
  Add(
   A=StaticSwitchParameter(True=Multiply(A=@n2.Fresnel Inverted, B=@n3.RGB), False=Constant()) {defaultValue=true, parameterName=UseOuterGlow}
   B=StaticSwitchParameter(True=Multiply(A=@n12, B=@n13.RGB), False=Constant()) {defaultValue=true, parameterName=UseSweep}
  )
  Constant4Vector() {constant=<rgba 1 0.575 0.15 0>}
  Constant() {r=1}
  Constant() {r=0.9}
  Constant() {r=0.1}
  n14 = Subtract(A=ComponentMask(None=TextureCoordinate()) {r=true}) {constB=0.45}
  n15 = LinearInterpolate(Alpha=Clamp(None=Ceil(None=@n14)))
  n16 = ActorPositionWS()
  n17 = Add(
   A=Divide(A=ComponentMask(None=@n16) {g=true, r=true}) {constB=512}
   B=Multiply(A=TextureCoordinate(), B=ScalarParameter() {defaultValue=0.25, parameterName=Pan Scale})
  )
  n18 = Panner(Coordinate=Add(A=@n17, B=Constant2Vector() {g=0.1134, r=0.4588})) {speedY=-0.2}
  n19 = Panner(Coordinate=Add(A=@n17, B=Constant2Vector() {g=0.7832, r=0.1343})) {speedX=0.015, speedY=-0.3}
  n20 = Divide(A=Add(
    A=TextureSample(UVs=@n18) {samplerType=SAMPLERTYPE_Normal, texture=/Game/Environments/Sets/Junkyard/Textures/T_Flag_N}.RGB
    B=TextureSample(UVs=@n19) {samplerType=SAMPLERTYPE_Normal, texture=&T_Flag_N}.RGB
   ))
  n21 = Constant3Vector() {constant=<rgba 1 0 0 0>}
  n22 = Multiply(
   A=Multiply(A=Multiply(A=@n20, B=@n21), B=ScalarParameter() {defaultValue=50, parameterName=Wind Strength}) {constB=50}
   B=ComponentMask(None=TextureCoordinate()) {g=true}
  )
  n23 = Multiply(A=@n15, B=@n22)
  n24 = Constant()
  FeatureLevelSwitch(Default=@n23, ES2=@n24, ES3_1=@n24, SM4_REMOVED=@n23, SM5=@n23, SM6=@n23)
  n25 = Multiply(A=@n15, B=@n22)
  LinearInterpolate(A=Constant3Vector() {constant=<rgba 0 0 1 0>}, B=@n20, Alpha=Clamp(None=Multiply(A=@n25) {constB=2.5}))
  n26 = ScalarParameter() {defaultValue=1, parameterName=Opacity}
  Subtract(A=Add(A=TextureSample() {texture=&T_NoisePack_01}.R), B=Multiply(A=OneMinus(None=@n26)) {constB=2})
 ]
}
