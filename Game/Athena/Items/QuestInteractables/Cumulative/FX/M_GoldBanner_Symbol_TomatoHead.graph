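# Lit material for the TomatoHead gold banner symbol quest interactable, with banner texture and color parameters and procedural cloth wind motion.
# Two panning flag normal samples, phase-shifted by actor position, are averaged and masked to X.
# Wind Strength and the V coordinate scale that into displacement, which feature-level switching zeroes on ES2 and ES3.1.
# The displacement also blends a flat normal toward the animated ripple normal.
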Material /Game/Athena/Items/QuestInteractables/Cumulative/FX/M_GoldBanner_Symbol_TomatoHead
graph {
 properties {pixelDepthOffsetMode=PDOM_Legacy, usedShadingModels=Default Lit}
 nodes [
  n1 = ActorPositionWS()
  n2 = Add(
   A=Divide(A=ComponentMask(None=@n1) {g=true, r=true}) {constB=512}
   B=Multiply(A=TextureCoordinate(), B=ScalarParameter() {defaultValue=0.25, parameterName=Pan Scale})
  )
  n3 = Panner(Coordinate=Add(A=@n2, B=Constant2Vector() {g=0.1134, r=0.4588})) {speedY=-0.2}
  n4 = Panner(Coordinate=Add(A=@n2, B=Constant2Vector() {g=0.7832, r=0.1343})) {speedX=0.015, speedY=-0.3}
  n5 = Divide(A=Add(
    A=TextureSample(UVs=@n3) {samplerType=SAMPLERTYPE_Normal, texture=/Game/Environments/Sets/Junkyard/Textures/T_Flag_N}.RGB
    B=TextureSample(UVs=@n4) {samplerType=SAMPLERTYPE_Normal, texture=&T_Flag_N}.RGB
   ))
  n6 = Constant3Vector() {constant=<rgba 1 0 0 0>}
  n7 = Multiply(
   A=Multiply(A=Multiply(A=@n5, B=@n6), B=ScalarParameter() {defaultValue=50, parameterName=Wind Strength}) {constB=50}
   B=ComponentMask(None=TextureCoordinate()) {g=true}
  )
  n8 = Constant()
  FeatureLevelSwitch(Default=@n7, ES2=@n8, ES3_1=@n8, SM4_REMOVED=@n7, SM5=@n7, SM6=@n7)
  LinearInterpolate(A=Constant3Vector() {constant=<rgba 0 0 1 0>}, B=@n5, Alpha=Clamp(None=Multiply(A=@n7) {constB=2.5}))
  TextureSampleParameter2D() {parameterName=BannerTexture, texture=/Game/2dAssets/Banners/Season5/T-Banners-Icons-Crown-L}
  VectorParameter() {parameterName=Color, defaultValue=<rgba 0.1058 0.0535 0.1719 0>}
 ]
}
